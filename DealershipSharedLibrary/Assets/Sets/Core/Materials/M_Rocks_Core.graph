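# Unlit FX material for rock mesh particles on Niagara, Nanite-enabled, that glows, dissolves and animates per piece.
# Emissive combines particle color masked by a noise-based Dissolve dynamic parameter with a Fresnel rim on faceted normals derived from position derivatives.
# Pivot Painter data and extra UV channels drive per-piece offset, time-based rotation and scaling through the Scale parameter.
# A distance mask offset by particle alpha blends vertices between object-origin collapse and a rerouted value.
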Material /DealershipSharedLibrary/Assets/Sets/Core/Materials/M_Rocks_Core
graph {
 properties {bUsedWithNanite=true, bUsedWithNiagaraMeshParticles=true, shadingModel=MSM_Unlit, usedShadingModels=Unlit}
 nodes [
  n1 = WorldPosition()
  MaterialExpressionObjectWorldPosition_15 = ObjectPositionWS()
  MaterialExpressionObjectWorldPosition_16 = ObjectPositionWS()
  n2 = Distance(A=WorldPosition().XYZ, B=@MaterialExpressionObjectWorldPosition_16)
  MaterialExpressionMeshEmitterVertexColor_11 = ParticleColor()
  MaterialExpressionConstantClamp_19 = Saturate(None=Multiply(A=Subtract(A=@n2, B=@MaterialExpressionMeshEmitterVertexColor_11.A), B=Constant() {r=0.01}))
  LinearInterpolate(
   A=Multiply(A=Subtract(A=@n1.XYZ, B=@MaterialExpressionObjectWorldPosition_15), B=Constant() {r=-1})
   B=NamedRerouteUsage()
   Alpha=@MaterialExpressionConstantClamp_19
  )
  MaterialExpressionMeshEmitterVertexColor_12 = ParticleColor()
  n3 = Multiply(A=ScalarParameter() {defaultValue=3, parameterName=MultIntensity}, B=@MaterialExpressionMeshEmitterVertexColor_12.RGB)
  MaterialExpressionMeshEmitterDynamicParameter_7 = DynamicParameter() {paramNames [Dissolve,Scale,Param3,Param4]}
  n4 = Add(
   A=TextureSample(UVs=TextureCoordinate() {uTiling=0.5, vTiling=0.5}) {texture=/DealershipSharedLibrary/FX/Textures/Noises/T_DLR_CloudDistort_A}.G
   B=ScalarParameter() {defaultValue=0.2, parameterName=edge_intensity}
  )
  MaterialExpressionConstantClamp_22 = Saturate(None=Multiply(
    A=Subtract(A=@n4, B=@MaterialExpressionMeshEmitterDynamicParameter_7.Dissolve)
    B=ScalarParameter() {defaultValue=5, parameterName=Mark Hardness}
   ))
  MaterialExpressionObjectWorldPosition_17 = ObjectPositionWS()
  n5 = Transform(None=Subtract(A=WorldPosition().XYZ, B=@MaterialExpressionObjectWorldPosition_17)) {transformSourceType=TRANSFORMSOURCE_Local}
  n6 = DDX(Value=@n5)
  n7 = DDY(Value=@n5)
  n8 = Transform(None=Normalize(VectorInput=CrossProduct(A=@n6, B=@n7))) {transformSourceType=TRANSFORMSOURCE_World, transformType=TRANSFORM_Local}
  n9 = Add(A=Constant() {r=0.1}, B=Multiply(A=Fresnel(Normal=@n8) {exponent=3}, B=Constant() {r=0.3}))
  MaterialFunctionCall(Emissive=Add(A=Multiply(A=@n3, B=@MaterialExpressionConstantClamp_22), B=Multiply(A=@n9, B=@MaterialExpressionMeshEmitterVertexColor_12.RGB))) {materialFunction=&MF_FX_EmissiveAdjustments}
  n10 = MaterialFunctionCall() {materialFunction=/Engine/Functions/Engine_MaterialFunctions02/PivotPainter/PivotPainter_PerObjectData}
  NamedRerouteDeclaration(None=@n10.Random Value Per Element) {name=PP_RandomValue, nodeColor=<rgba 1 0.75 0 1>}
  NamedRerouteDeclaration(None=@n10.Pivot Position) {name=PP_PivotPosition, nodeColor=<rgba 0.4297 0 1 1>}
  NamedRerouteDeclaration(None=ComponentMask(None=TextureCoordinate() {coordinateIndex=3}) {g=true}) {name=RandomValueUVs, nodeColor=<rgba 0 1 0.1094 1>}
  n11 = Multiply(
   A=ScalarParameter() {defaultValue=1, desc=Note: Multiply the real value in UE3 by 5 and this is the new RotationValue, parameterName=Rotation WPO}
   B=Time()
  )
  n12 = NamedRerouteUsage()
  NamedRerouteDeclaration(None=Multiply(A=Multiply(A=@n11, B=@n12), B=ScalarParameter() {defaultValue=0.2, parameterName=RotateAmount})) {name=RotationAngle, nodeColor=<rgba 1 0 0.2109 1>}
  n13 = Constant()
  n14 = NamedRerouteUsage()
  n15 = ScalarParameter() {controlType=Enumeration, enumeration=/DealershipSharedLibrary/FX/Materials/Enums/Enum_DLR_TrueFalse, parameterName=UseRandomScale?}
  n16 = Saturate(None=@n15)
  n17 = Multiply(
   A=LinearInterpolate(A=Constant() {r=-1}, B=LinearInterpolate(A=@n13, B=@n14, Alpha=@n16), Alpha=NamedRerouteUsage())
   B=Subtract(A=WorldPosition().XYZ, B=NamedRerouteUsage())
  )
  NamedRerouteDeclaration(None=Add(A=@n17, B=WorldPosition().XYZ)) {name=RotateAX Position, nodeColor=<rgba 1 0.1875 0 1>}
  NamedRerouteDeclaration(None=@MaterialExpressionMeshEmitterDynamicParameter_7.Scale) {name=DYN Scale, nodeColor=<rgba 0.2656 0 1 1>}
  n18 = TextureCoordinate() {coordinateIndex=2}
  n19 = ComponentMask(None=TextureCoordinate() {coordinateIndex=3}) {r=true}
  n20 = Multiply(
   A=Transform(None=AppendVector(A=@n18, B=@n19)) {transformSourceType=TRANSFORMSOURCE_Local}
   B=Add(A=ScalarParameter() {defaultValue=1, parameterName=OldAlpha}, B=NamedRerouteUsage())
  )
  n21 = Constant3Vector() {constant=<rgba 0 1 0 0>}
  n22 = Constant3Vector() {constant=<rgba 1 -1 0 0>}
  n23 = Floor(None=Multiply(A=Constant() {r=10}, B=NamedRerouteUsage()))
  n24 = RotateAboutAxis(
   NormalizedRotationAxis=Normalize(VectorInput=LinearInterpolate(A=@n21, B=@n22, Alpha=@n23))
   RotationAngle=NamedRerouteUsage()
   PivotPoint=NamedRerouteUsage()
   Position=NamedRerouteUsage()
  )
  NamedRerouteDeclaration(None=Add(A=Add(A=@n20, B=@n24), B=@n17)) {name=WPORotation, nodeColor=<rgba 1 0.0234 0 1>}
 ]
}
